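# Animation Blueprint for the BeanDog bow skeleton, combining montage playback with physics-driven secondary motion.
# The AnimGraph feeds the reference pose through DefaultSlot, then applies a RigidBody node in component space before output.
# Each update copies the RigidBodyAlpha curve into that node's alpha, defaulting to full simulation, so animations can fade physics out through the curve.
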AnimBlueprint /WildEstateWeapons/Assets/BeanDog_Bow/Animation/ABP_BeanDog_Bow
target_skeleton=/WildEstateWeapons/Assets/BeanDog_Bow/Meshes/BeanDog_Bow_Skeleton
graphs [
 AnimGraph {nodes [
   n1 = Local To Component(LocalPose=Slot 'DefaultSlot'(Source=Local Space Ref Pose()))
   n2 = Get RigidBodyAlpha() -> (RigidBodyAlpha:float)
   Output Pose(Result=Component To Local(ComponentPose=RigidBody(ComponentPose=@n1, Alpha=@n2) {alphaBoolBlend {blendOption=Linear}}))
  ]}
 EventGraph {nodes [
   Event BlueprintUpdateAnimation()
   > Set RigidBodyAlpha(RigidBodyAlpha:float=GetCurveValueWithDefault!(self:Anim Instance Object Reference, CurveName:Name=RigidBodyAlpha, DefaultValue:float=1) -> (OutValue:float, ReturnValue:bool)) -> (Output_Get:float)
  ]}
]
variables [RigidBodyAlpha {type=number, default=1}]
