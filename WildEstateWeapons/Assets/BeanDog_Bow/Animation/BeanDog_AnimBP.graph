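# Animation Blueprint for the BeanDog bow/slingshot weapon, animating its rest and charging poses and string pull.
# A state machine switches between Default and Charging (normal or fast entry), feeding DefaultSlot and a PreIK cached pose.
# Transform Bone nodes then pin string_pull and arrow_nock to a world-space string transform, offsetting string_pull, unless bIkHandR_Off bypasses them.
# Curve-driven wing rotation nodes in Charging are unconnected, and the EventGraph is empty.
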AnimBlueprint /WildEstateWeapons/Assets/BeanDog_Bow/Animation/BeanDog_AnimBP
target_skeleton=/WildEstateWeapons/Assets/BeanDog_Bow/Meshes/BeanDog_Bow_Skeleton
graphs [
 AnimGraph {nodes [
   n1 = Get StringRotation() -> (StringRotation:Rotator)
   n2 = Get StringLocation() -> (StringLocation:Vector)
   n3 = "Transform (Modify) Bone - Bone: string_pull"!(
    ComponentPose=Local To Component(LocalPose=Use cached pose 'PreIK'())
    bAlphaBoolEnabled:bool=Get bStringIK_On() -> (bStringIK_On:bool)
    Translation=@n2
    Rotation=@n1
    Scale="1,1,1"
   ) -> (Pose) {
    alphaInputType=Bool
    rotationMode=BMM_Replace
    rotationSpace=BCS_WorldSpace
    translationMode=BMM_Replace
    translationSpace=BCS_WorldSpace
    alphaBoolBlend {blendInTime=0.1, blendOption=Linear}
    boneToModify {boneName=string_pull}
   }
   n4 = "Transform (Modify) Bone - Bone: arrow_nock"!(ComponentPose=@n3, bAlphaBoolEnabled:bool=True, Translation=@n2, Rotation=@n1, Scale="1,1,1") -> (Pose) {
    alphaInputType=Bool
    rotationMode=BMM_Replace
    rotationSpace=BCS_WorldSpace
    translationMode=BMM_Replace
    translationSpace=BCS_WorldSpace
    alphaBoolBlend {blendInTime=0.1, blendOption=Linear}
    boneToModify {boneName=arrow_nock}
   }
   n5 = "Transform (Modify) Bone - Bone: string_pull"(ComponentPose=@n4, bAlphaBoolEnabled=Get bStringIK_On(), Translation="6,0,2", Rotation="-90,0,0", Scale="1,1,1") {
    alphaInputType=Bool
    rotationMode=BMM_Additive
    rotationSpace=BCS_BoneSpace
    translationMode=BMM_Additive
    translationSpace=BCS_BoneSpace
    alphaBoolBlend {blendInTime=0.1, blendOption=Linear}
    boneToModify {boneName=string_pull}
    rotation {pitch=-90}
    translation {x=6, z=2}
   }
   Output Pose(Result=Blend Poses by bool(
     BlendPose_0=Use cached pose 'PreIK'()
     BlendPose_1=Component To Local(ComponentPose=@n5)
     bActiveValue=Get bIkHandR_Off() -> (bIkHandR_Off:bool)
    )) {layerGroup=None}
   AnimGraphNode_StateMachine_0 = SlingShotStateMachine() {is=title AnimGraphNode_StateMachine}
   Save cached pose 'PreIK'(Pose=Slot 'DefaultSlot'(Source=@AnimGraphNode_StateMachine_0)) {cacheName=PreIK}
  ]}
 SlingShotStateMachine {
  entry=Default
  path=AnimGraph.AnimGraphNode_StateMachine_0.SlingShotStateMachine
  ids=id
  nodes [
   AnimStateNode_0 = Default() {is=title AnimStateNode}
   AnimStateNode_1 = Charging() {is=title AnimStateNode}
   AnimStateTransitionNode_0 = AnimStateTransitionNode() {crossfadeDuration=0.55}
   AnimStateTransitionNode_1 = AnimStateTransitionNode() {crossfadeDuration=0.1}
   AnimStateTransitionNode_4 = AnimStateTransitionNode() {crossfadeDuration=0}
  ]
  links [
   {
    dst=Default
    rule=AnimGraph.AnimGraphNode_StateMachine_0.SlingShotStateMachine.AnimStateTransitionNode_4.Transition
    src=Charging
    via=AnimStateTransitionNode_4
   }
   {
    dst=Charging
    rule=AnimGraph.AnimGraphNode_StateMachine_0.SlingShotStateMachine.AnimStateTransitionNode_0.Transition
    src=Default
    via=AnimStateTransitionNode_0
   }
   {
    dst=Charging
    rule=AnimGraph.AnimGraphNode_StateMachine_0.SlingShotStateMachine.AnimStateTransitionNode_1.Transition
    src=Default
    via=AnimStateTransitionNode_1
   }
  ]
 }
 Default {
  path=AnimGraph.AnimGraphNode_StateMachine_0.SlingShotStateMachine.AnimStateNode_0.Default
  nodes [
   Sequence Player 'WildEstate_BeanDog_Slingshot_BasePose_WEP'() {bLoopAnimation=false, sequence=&/WildEstate_BeanDog_Slingshot_BasePose_WEP}
   Output Animation Pose(Result=Blend Poses by bool(
     BlendPose_0=Sequence Player 'WildEstate_BeanDog_Slingshot_Relaxed_Pose_WEP'() {bLoopAnimation=false, sequence=&/WildEstate_BeanDog_Slingshot_Relaxed_Pose_WEP}
     BlendPose_1="Sequence Evaluator (by time) 'WildEstate_BeanDog_Slingshot_BasePose_WEP'"!(ExplicitTime:float=0) -> (Pose:Pose Link Structure) {is=title AnimGraphNode_SequenceEvaluator, sequence=&WildEstate_BeanDog_Slingshot_BasePose_WEP}
     BlendTime_0=0.15
     BlendTime_1=0.3
     bActiveValue=Get bShouldPlayDefaultPose() -> (bShouldPlayDefaultPose:bool)
    ) {blendType=Linear})
  ]
 }
 Charging {
  path=AnimGraph.AnimGraphNode_StateMachine_0.SlingShotStateMachine.AnimStateNode_1.Charging
  nodes [
   n1 = GetCurveValue(CurveName=Wing_Charging)
   n2 = "Transform (Modify) Bone - Bone: bow_wing_top"(Alpha=1, Translation="0,0,0", Rotation=MakeRotator(Roll=@n1), Scale="1,1,1") {
    rotationMode=BMM_Additive
    rotationSpace=BCS_BoneSpace
    scaleMode=BMM_Additive
    translationMode=BMM_Additive
    alphaBoolBlend {blendOption=Linear}
    boneToModify {boneName=bow_wing_top}
   }
   n3 = "Transform (Modify) Bone - Bone: bow_wing_bott"(ComponentPose=@n2, Alpha=1, Translation="0,0,0", Rotation=MakeRotator(Roll=float * float(A=@n1, B=-1)), Scale="1,1,1") {
    rotationMode=BMM_Additive
    rotationSpace=BCS_BoneSpace
    scaleMode=BMM_Additive
    translationMode=BMM_Additive
    alphaBoolBlend {blendOption=Linear}
    boneToModify {boneName=bow_wing_bott}
   }
   Component To Local(ComponentPose=@n3)
   n4 = Get ChargePlayRate() -> (ChargePlayRate:float)
   Output Animation Pose(Result=Two Way Blend!(
     A=Sequence Player 'WildEstate_BeanDog_Slingshot_ChargingIntro_WEP'!(PlayRate:float=@n4) -> (Pose) {bLoopAnimation=false, sequence=&/WildEstate_BeanDog_Slingshot_ChargingIntro_WEP}
     B=Sequence Player 'WildEstate_BeanDog_Slingshot_ChargingLoop_WEP'() {sequence=&/WildEstate_BeanDog_Slingshot_ChargingLoop_WEP}
     bAlphaBoolEnabled:bool=Get bExitCharging() -> (bExitCharging:bool)
    ) -> (Pose) {blendNode {alphaInputType=Bool, bResetChildOnActivation=true, alphaBoolBlend {blendInTime=0.5, blendOption=Linear}}})
  ]
 }
 Transition {
  path=AnimGraph.AnimGraphNode_StateMachine_0.SlingShotStateMachine.AnimStateTransitionNode_0.Transition
  nodes [Result(bCanEnterTransition=Get bEnterChargingNormalTransition() -> (bEnterChargingNormalTransition:bool))]
 }
 Transition {
  path=AnimGraph.AnimGraphNode_StateMachine_0.SlingShotStateMachine.AnimStateTransitionNode_1.Transition
  nodes [Result(bCanEnterTransition=Get bEnterChargingFastTransition() -> (bEnterChargingFastTransition:bool))]
 }
 Transition {
  path=AnimGraph.AnimGraphNode_StateMachine_0.SlingShotStateMachine.AnimStateTransitionNode_4.Transition
  nodes [Result(bCanEnterTransition=NOT Boolean(A=Get bIsBowCharging() -> (bIsBowCharging:bool)))]
 }
 EventGraph {links [], nodes []}
]
variables []
