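# Physics asset for the BeanDog bow skeletal mesh, giving the handle and string bones capsule bodies for physics.
# On each side, a base segment hinges off the handle and a flex segment off the base, with limited swing and base twist held at zero.
# Both flex segments join the central string_pull body with free angular motion, closing the string into a loop through the handle.
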PhysicsAsset /WildEstateWeapons/Assets/BeanDog_Bow/Meshes/BeanDog_Bow_PhysicsAsset
bodies {
 handle [{shapeType=Capsule, length=15, radius=2, center=<0 0 3>}]
 string_main_base_l [{
   shapeType=Capsule
   length=5
   radius=0.505
   shapeName=string_main_base_l_capsule
   center=<-3.0573 0.0003 0.0255>
   rotation=<rot 0 -90.2969 -89.5557>
  }]
 string_main_base_r [{
   shapeType=Capsule
   length=5
   radius=0.505
   shapeName=string_main_base_r_capsule
   center=<3.0573 0.0005 0.0256>
   rotation=<rot 0 90.2969 -89.5557>
  }]
 string_main_flex_l [{
   shapeType=Capsule
   length=5
   radius=0.505
   shapeName=string_main_flex_l_capsule
   center=<-3.0155 -0.0208 0.1202>
   rotation=<rot 0 -89.8902 -89.1098>
  }]
 string_main_flex_r [{
   shapeType=Capsule
   length=5
   radius=0.505
   shapeName=string_main_flex_r_capsule
   center=<3.0155 -0.0206 0.1202>
   rotation=<rot 0 89.8902 -89.1098>
  }]
 string_pull [{shapeType=Capsule, length=2, radius=1, shapeName=string_pull_capsule, center=<-0.4966 -0.0001 -0.0337>, rotation=<rot 0 0 90>}]
}
constraints [
 {bone1Name=string_main_base_l, bone2Name=handle, swing1LimitDegrees=15, swing2LimitDegrees=45, twistLimitDegrees=0}
 {bone1Name=string_main_flex_l, bone2Name=string_main_base_l, swing1LimitDegrees=15, swing2LimitDegrees=45, twistLimitDegrees=45}
 {bone1Name=string_main_base_r, bone2Name=handle, swing1LimitDegrees=15, swing2LimitDegrees=45, twistLimitDegrees=0}
 {bone1Name=string_main_flex_r, bone2Name=string_main_base_r, swing1LimitDegrees=15, swing2LimitDegrees=45, twistLimitDegrees=45}
 {
  bone1Name=string_pull
  bone2Name=string_main_flex_l
  swing1LimitDegrees=45
  swing1Motion=Free
  swing2LimitDegrees=45
  swing2Motion=Free
  twistLimitDegrees=45
  twistMotion=Free
 }
 {
  bone1Name=string_pull
  bone2Name=string_main_flex_r
  swing1LimitDegrees=45
  swing1Motion=Free
  swing2LimitDegrees=45
  swing2Motion=Free
  twistLimitDegrees=45
  twistMotion=Free
 }
]
